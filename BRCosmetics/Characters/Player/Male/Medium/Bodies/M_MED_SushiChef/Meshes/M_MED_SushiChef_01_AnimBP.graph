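# Post-copy secondary-motion AnimBlueprint for the male medium Sushi Chef body, adding physics sway to the outfit's apron.
# It copies the attached parent mesh's pose, then chains three root-relative Anim Dynamics nodes for front-left, front-right and rear apron bones.
# Thigh-driven spherical limits constrain the front flaps.
# The event graph sets gravity and force variables from skydiving, parachute and vertical-velocity state, but no shown AnimGraph node reads them.
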AnimBlueprint /BRCosmetics/Characters/Player/Male/Medium/Bodies/M_MED_SushiChef/Meshes/M_MED_SushiChef_01_AnimBP
target_skeleton=&/M_MED_SushiChef_01_Skeleton
graphs [
 AnimGraph {nodes [
   n1 = Anim Dynamics - Bone: dyn_front_apron_02_l(ComponentPose=Local To Component(LocalPose=Copy Pose From Mesh() {bUseAttachedParent=true}), Alpha=0.7, ExternalForce="0,0,0") {
    angularDampingOverride=0.9999
    bGravityOverrideInSimSpace=true
    bOverrideAngularDamping=true
    gravityScale=0.35
    simulationSpace=RootRelative
    alphaBoolBlend {blendOption=Linear}
    boundBone {boneName=dyn_front_apron_02_l}
    componentAppliedLinearAccClamp=<1000 1000 1000>
    componentLinearAccScale=<0.1 0.1 0.1>
    componentLinearVelScale=<0.1 0.1 0.1>
    physicsBodyDefinitions [{
      collisionType=CustomSphere
      sphereCollisionRadius=1
      boundBone {boneName=dyn_front_apron_02_l}
      boxExtents=<10 30 30>
      constraintSetup {angularLimitsMax=<0 3 0>, angularLimitsMin=<0 -20 -22>, angularTarget=<0 0 0>}
      localJointOffset=<-10 0 0>
     }]
    sphericalLimits [{limitRadius=9, drivingBone {boneName=thigh_l}, sphereLocalOffset=<-24 0 0>}]
   }
   n2 = Anim Dynamics - Bone: dyn_front_apron_02_r(ComponentPose=@n1, Alpha=0.7, ExternalForce="0,0,0") {
    angularDampingOverride=0.9999
    bGravityOverrideInSimSpace=true
    bOverrideAngularDamping=true
    gravityScale=0.35
    simulationSpace=RootRelative
    alphaBoolBlend {blendOption=Linear}
    boundBone {boneName=dyn_front_apron_02_r}
    componentAppliedLinearAccClamp=<1000 1000 1000>
    componentLinearAccScale=<0.1 0.1 0.1>
    componentLinearVelScale=<0.1 0.1 0.1>
    physicsBodyDefinitions [{
      collisionType=CustomSphere
      sphereCollisionRadius=1
      boundBone {boneName=dyn_front_apron_02_r}
      boxExtents=<10 30 30>
      constraintSetup {angularLimitsMax=<0 3 0>, angularLimitsMin=<0 -20 -22>, angularTarget=<0 0 0>}
      localJointOffset=<10 0 0>
     }]
    sphericalLimits [{limitRadius=11, drivingBone {boneName=thigh_r}, sphereLocalOffset=<24 0 0>}]
   }
   n3 = Anim Dynamics - Bone: dyn_rear_apron(ComponentPose=@n2, Alpha=0.65, ExternalForce="0,0,0") {
    angularDampingOverride=0.999
    bGravityOverrideInSimSpace=true
    bOverrideAngularDamping=true
    gravityScale=0.33
    simulationSpace=RootRelative
    alphaBoolBlend {blendOption=Linear}
    boundBone {boneName=dyn_rear_apron}
    componentAppliedLinearAccClamp=<1000 1000 1100>
    componentLinearAccScale=<0.3 1 3>
    componentLinearVelScale=<0.3 0.3 0.3>
    physicsBodyDefinitions [{
      sphereCollisionRadius=0
      boundBone {boneName=dyn_rear_apron}
      boxExtents=<30 20 15>
      constraintSetup {angularTargetAxis=AxisZ, twistAxis=AxisZ, angularLimitsMax=<0 10 0>, angularLimitsMin=<-40 -10 0>, angularTarget=<0 0 0>}
      localJointOffset=<0 0 -10>
     }]
   }
   Output Pose(Result=Component To Local(ComponentPose=@n3)) {layerGroup=None}
  ]}
 EventGraph {nodes [
   Event BlueprintUpdateAnimation()
   > Get!(execute:Exec) -> (then:Exec, else:Exec, Skel_Mesh:Skeletal Mesh Component Object Reference)
   > Get!(execute) -> (then, else, Anim_Inst:Anim Instance Object Reference)
   > Sequence!(execute) -> (then_0, then_1, then_2:Exec, then_3:Exec, then_4:Exec, then_5:Exec, then_6:Exec)
   > n1 = Branch(Condition=Get bIsSkyDiving())
   > n2 = Branch(Condition=Get bIsParachuteOpened())
   > Set gravity_override(gravity_override="0, 0,-800")
   n3 = Break Fort Anim Input Skydiving External Force(FortAnimInput_SkydivingExternalForce=Get SkydivingExternalForceInput()) -> (
    bUseSkydivingVectorForce=~
    SkydivingVectorForce=~
    HeadToPelvisDirection=~
    FloatingMultiplier=~
    FloatingAdditive=~
    DivingMultiplier=~
    DivingAdditive=~
    ParachutingMultiplier=~
    ParachutingAdditive=~
   )
   > Set vector_force(vector_force:Vector=@n3.SkydivingVectorForce) -> (Output_Get:Vector)
   n4 = Branch(execute=@n2.else, Condition="float > float"(A=Get PawnVelocityZ(), B=50) {is=title K2Node_CallFunction})
   > Set gravity_override(gravity_override="0, 0,-300")
   > Set vector_force(vector_force=@n3.SkydivingVectorForce)
   Set gravity_override(execute=@n4.else, gravity_override="0, 0,-300")
   > Set vector_force(vector_force=@n3.SkydivingVectorForce)
   > Branch(Condition=Get bCanPlayCustomAnimations() -> (bCanPlayCustomAnimations:bool))
   > Set gravity_override(gravity_override="0, 0,150")
   Set gravity_override(execute=@n1.else, gravity_override="0, 0,-920")
   > Set vector_force(vector_force="0, 0, 0")
   Event BlueprintInitializeAnimation()
   > Set Skel_Mesh(Skel_Mesh=GetOwnerSkeletalMeshComponent())
   > n5 = Get!(execute) -> (then, else, Skel_Mesh:Skeletal Mesh Component Object Reference)
   > Set Anim_Inst(Anim_Inst:Anim Instance Object Reference=GetAnimInstance(self=@n5)) -> (Output_Get:Anim Instance Object Reference)
  ]}
]
variables [
 Anim_Inst {type=AnimInstance, default=null}
 gravity_override {type=Vector}
 Skel_Mesh {type=SkeletalMeshComponent}
 vector_force {type=Vector}
]
