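# Cloth physics asset for the male medium SushiChef player body, named as the collision setup for its cloth.
# Capsules cover the thighs, calves and front apron bone, while the pelvis carries five capsules, including ones named for spine, back and front.
# Constraints link calves to thighs, and thighs and apron to pelvis, with swing and twist all free, so they impose no angular limits.
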PhysicsAsset /BRCosmetics/Characters/Player/Male/Medium/Bodies/M_MED_SushiChef/Meshes/M_MED_SushiChef_Cloth_Physics
bodies {
 calf_l [{shapeType=Capsule, length=25.2751, radius=8.4188, center=<-12.9799 -1.4829 -0.2423>, rotation=<rot -0.6019 -91.9574 -88.736>}]
 calf_r [{shapeType=Capsule, length=25.0857, radius=8.5818, center=<13.2443 0.7863 -0.0731>, rotation=<rot -0.1465 89.0026 -88.8377>}]
 dyn_front_apron_c [{shapeType=Capsule, length=9.27, radius=9.2349, center=<-12.3155 -8.9874 0.0052>, rotation=<rot 90 0 0>}]
 pelvis [
  {shapeType=Capsule, length=9.6204, radius=8.0151, shapeName=bk_01, center=<-2.9046 -2.8793 0.1002>, rotation=<rot -0.0005 177.9471 177.9471>}
  {shapeType=Capsule, length=41.5286, radius=9.9856, shapeName=spine_01, center=<23.4252 -1.6037 0.0007>, rotation=<rot -90 0 -4.106>}
  {shapeType=Capsule, length=10.2025, radius=8.9666, shapeName=bk_02, center=<-8.0492 -3.1655 0.1296>, rotation=<rot -0.0005 177.9471 177.9471>}
  {shapeType=Capsule, length=10.0605, radius=8.9666, shapeName=ft_01, center=<-8.564 3.3559 -0.0913>, rotation=<rot -0.0005 177.9471 177.9471>}
  {shapeType=Capsule, length=10.6153, radius=9.1958, center=<-12.5423 3.5609 -0.0805>, rotation=<rot -0.0005 177.9471 177.9471>}
 ]
 thigh_l [{shapeType=Capsule, length=27.723, radius=10.577, center=<-20.4987 0.865 2.7915>, rotation=<rot 90 3.5763 8.5763>}]
 thigh_r [{shapeType=Capsule, length=27.723, radius=10.577, center=<20.4986 -0.8656 -1.387>, rotation=<rot 90 3.5763 8.5763>}]
}
constraints [
 {
  bone1Name=calf_l
  bone2Name=thigh_l
  swing1LimitDegrees=45
  swing1Motion=Free
  swing2LimitDegrees=45
  swing2Motion=Free
  twistLimitDegrees=45
  twistMotion=Free
 }
 {
  bone1Name=calf_r
  bone2Name=thigh_r
  swing1LimitDegrees=45
  swing1Motion=Free
  swing2LimitDegrees=45
  swing2Motion=Free
  twistLimitDegrees=45
  twistMotion=Free
 }
 {
  bone1Name=dyn_front_apron_c
  bone2Name=pelvis
  swing1LimitDegrees=45
  swing1Motion=Free
  swing2LimitDegrees=45
  swing2Motion=Free
  twistLimitDegrees=45
  twistMotion=Free
 }
 {
  bone1Name=thigh_l
  bone2Name=pelvis
  swing1LimitDegrees=45
  swing1Motion=Free
  swing2LimitDegrees=45
  swing2Motion=Free
  twistLimitDegrees=45
  twistMotion=Free
 }
 {
  bone1Name=thigh_r
  bone2Name=pelvis
  swing1LimitDegrees=45
  swing1Motion=Free
  swing2LimitDegrees=45
  swing2Motion=Free
  twistLimitDegrees=45
  twistMotion=Free
 }
]
